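; BP_Hera_Plant_JacobsLadder_A_Snow is a foliage Blueprint, by its name and path the snow variant of Jacob's Ladder plant A in Asteria's chaparral plant set.
; Its construction script only calls the parent implementation.
; BeginPlay, ActorBeginOverlap and Tick are disabled placeholder events, so the graphs shown add no behavior of their own.
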(event EventBeginPlay
  (|CallParentFunction) ; Questo nodo è disabilitato e non verrà chiamato.\nTrascina i pin per creare funzionalità.
)

(event Collision|EventActorBeginOverlap
  (params (OtherActor :type Actor%20Object%20Reference))
  (|CallParentFunction OtherActor) ; Questo nodo è disabilitato e non verrà chiamato.\nTrascina i pin per creare funzionalità.
)

(event EventTick
  (params (DeltaSeconds :type Float%20%28single-precision%29))
  (|CallParentFunction DeltaSeconds) ; Questo nodo è disabilitato e non verrà chiamato.\nTrascina i pin per creare funzionalità.
)

(fn ConstructionScript
  (|CallParentFunction))
